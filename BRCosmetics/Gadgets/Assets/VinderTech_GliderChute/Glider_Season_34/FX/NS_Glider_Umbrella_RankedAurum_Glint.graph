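# Looping glint sparkle effect for the Season 34 Ranked Aurum umbrella glider, tinted by the player's ranked progression.
# Flare_loop pins small flares near five skeletal mesh bones or sockets using particle partitions, while Flare_loop001 scatters denser flares in a flattened sphere.
# Both pulse sprite size over life and multiply a system coin color by per-rank array tints, with rank-specific corrections at night.
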NiagaraSystem /BRCosmetics/Gadgets/Assets/VinderTech_GliderChute/Glider_Season_34/FX/NS_Glider_Umbrella_RankedAurum_Glint
dependencies {
 usedDataInterfaces [
  &NiagaraDataInterfaceCurve
  &NiagaraDataInterfaceArrayFloat3
  &NiagaraDataInterfaceSkeletalMesh
  &NiagaraDataInterfaceArrayColor
 ]
 usedDynamicInputs [&SelectLinearColorFromArray]
 usedModules [
  &SystemState
  &self:SystemSpawnScript.SystemScriptSource.SystemScriptGraph.NiagaraNodeAssignment_1.SetVariables_#1
  &self:SystemSpawnScript.SystemScriptSource.SystemScriptGraph.NiagaraNodeAssignment_3.SetVariables_#2
  &self:SystemSpawnScript.SystemScriptSource.SystemScriptGraph.NiagaraNodeAssignment_0.SetVariables_#3
 ]
}
emitters {
 Flare_loop {
  emitterUpdateScript [
   &EmitterState(
    Delay First Loop Only=true
    Inactive Response=::NewEnumerator1
    Life Cycle Mode=::NewEnumerator1
    Loop Delay=0.8
    Loop Duration=1
    UseLoopDelay=true
   )
   &SpawnRate(SpawnRate=10, Use Spawn Probability=false)
  ]
  particleSpawnScript [
   &PartitionParticles(Number Of Partitions=5)
   &InitializeParticle(
    Color=$System.CoinColor
    Lifetime Max=1
    Lifetime Min=0.5
    Lifetime Mode=::NewEnumerator1
    Sprite Size Max=<8 8>
    Sprite Size Mode=::NewEnumerator2
    UsePositionOffset=false
    Write Material Random=true
   )
   &CameraOffset(Camera Offset Amount=5)
  ]
  particleUpdateScript [
   &ParticleState()
   &ScaleSpriteSize(
    Uniform Curve Scale=&MakeCustomFloatFromBool(Bool=$User.bIsFrontEndPreview, True Float=4, False Float=4)
    Uniform Curve Sprite Scale=DataInterface() {propertyValues {Curve {keys [{tangentMode=RCTM_User, time=0, value=0}, <at 0.2 1>, <at 0.3 0.4>, {tangentMode=RCTM_User, time=1, value=0}]}}}
   )
   /SetVariables(Particles.Color002=&MultiplyLinearColor(
     Linear Color=$System.CoinColor
     Scale Factor=&MakeFloat4FromVectorAndFloat(
      "Vector (RGB)"=&Multiply_Vector(
       A=&SelectVectorFromArray(
        Direct Array Index=$System.RankedProgression
        Vector Selection Array=DataInterface() {propertyValues {FloatData [<3 2 0.3>, <1 1 0.5>, <1 1 0.2>, <1 1 1>, <1 0.5 1>, <0.5 1 1>, <2 0.2 0.1>, <1 0.1 1>]}}
       )
       B=&FN_Select_Vector_from_Time_Of_Day(
        Morning=<1 1 1>
        Day=<1 1 1>
        Evening=<1 1 1>
        Night=&SelectVectorFromArray(
         Direct Array Index=$System.RankedProgression
         Vector Selection Array=DataInterface() {propertyValues {FloatData [<0.5 0.5 1>, <1 1 1>, <0.5 0.5 1>, <1 1 1>, <1 1 1>, <1 1 1>, <0.4 5 1>, <1 1 1>]}}
        )
       )
      )
      "Float (Alpha)"=0.03
     )
    )) {isSetParameters=true}
   &SkeletalMeshLocation(
    Bone / Socket Index=$Particles.PartitionParticles.Partition
    Bone Sampling Mode=::NewEnumerator4
    MaskSpawnGroup=false
    Sampled Position Offset=&Add_Vector(
     A=&Add_Vector(A=<0 0 0>, B=&RandomRangeVector(Minimum=<1 2 2>, Maximum=<-1 -2 -2>))
     B=&SelectVectorFromArray(
      Direct Array Index=$Particles.PartitionParticles.Partition
      Vector Selection Array=DataInterface() {propertyValues {FloatData [<5 0 0>, <3.5 0 0>, <4 0 0>, <3 0 0>, <4 0 0>]}}
     )
    )
    Skeletal Mesh=$System.Skeletal Mesh
   )
   &ShapeLocation(MaskSpawnGroup=false, Offset=<0 0 10>, Offset Mode=::NewEnumerator0, Override Seed=false, Sphere Radius=0, UseOwnerScale=true) {enabled=false, module=ShapeLocation001}
   &ScaleColor(
    Scale RGB=&Multiply_Vector(
     A=&SelectVectorFromArray(
      Direct Array Index=$System.RankedProgression
      Vector Selection Array=DataInterface() {propertyValues {FloatData [<3 2 1>, <1 1 0.5>, <1 1 0.2>, <1 1 0.5>, <1 0.5 1>, <0.5 1 1>, <2 0.2 0.1>, <1 0.1 1>]}}
     )
     B=&FN_Select_Vector_from_Time_Of_Day(
      Morning=<1 1 1>
      Day=<1 1 1>
      Evening=<1 1 1>
      Night=&SelectVectorFromArray(
       Direct Array Index=$System.RankedProgression
       Vector Selection Array=DataInterface() {propertyValues {FloatData [<0.5 0.5 1>, <1 1 1>, <0.5 0.5 1>, <1 1 1>, <1 1 1>, <1 1 1>, <0.4 5 1>, <1 1 1>]}}
      )
     )
    )
    ScaleA=true
    ScaleRGB=true
   )
  ]
  rendererClasses [&NiagaraSpriteRendererProperties]
  renderers [
   {rendererClass=&NiagaraSpriteRendererProperties}
   {rendererClass=&NiagaraSpriteRendererProperties, rendererIndex=1}
  ]
 }
 Flare_loop001 {
  emitterUpdateScript [
   &EmitterState(
    Delay First Loop Only=true
    Inactive Response=::NewEnumerator1
    Life Cycle Mode=::NewEnumerator1
    Loop Delay=0.8
    Loop Duration=1
    UseLoopDelay=true
   )
   &SpawnRate(SpawnRate=40, Use Spawn Probability=false)
  ]
  particleSpawnScript [
   &PartitionParticles(Number Of Partitions=5) {enabled=false}
   &InitializeParticle(
    Color=$System.CoinColor
    Lifetime Max=1
    Lifetime Min=0.5
    Lifetime Mode=::NewEnumerator1
    Sprite Size Mode=::NewEnumerator2
    UsePositionOffset=false
    Write Material Random=true
   )
   &ShapeLocation(
    Hemisphere Distribution=<1 0.5>
    MaskSpawnGroup=false
    Non Uniform Scale=<1 1 0.45>
    Offset=<0 0 8>
    Offset Mode=::NewEnumerator0
    Override Seed=false
    Sphere Radius=50
    Sphere Surface Distribution=0.8
    UseOwnerScale=true
   ) {module=ShapeLocation001}
   &CameraOffset(Camera Offset Amount=1)
  ]
  particleUpdateScript [
   &ParticleState()
   &ScaleSpriteSize(
    Uniform Curve Scale=&MakeCustomFloatFromBool(Bool=$User.bIsFrontEndPreview, True Float=4, False Float=5)
    Uniform Curve Sprite Scale=DataInterface() {propertyValues {Curve {keys [{tangentMode=RCTM_User, time=0, value=0}, <at 0.2 1>, <at 0.3 0.4>, {tangentMode=RCTM_User, time=1, value=0}]}}}
   )
   /SetVariables(Particles.Color002=&MultiplyLinearColor(
     Linear Color=$System.CoinColor
     Scale Factor=&MakeFloat4FromVectorAndFloat(
      "Vector (RGB)"=&Multiply_Vector(
       A=&SelectVectorFromArray(
        Direct Array Index=$System.RankedProgression
        Vector Selection Array=DataInterface() {propertyValues {FloatData [<3 2 0.3>, <1 1 0.5>, <1 1 0.2>, <1 1 1>, <1 0.5 1>, <0.5 1 1>, <2 0.2 0.1>, <1 0.1 1>]}}
       )
       B=&FN_Select_Vector_from_Time_Of_Day(
        Morning=<1 1 1>
        Day=<1 1 1>
        Evening=<1 1 1>
        Night=&SelectVectorFromArray(
         Direct Array Index=$System.RankedProgression
         Vector Selection Array=DataInterface() {propertyValues {FloatData [<0.5 0.5 1>, <1 1 1>, <0.5 0.5 1>, <1 1 1>, <1 1 1>, <1 1 1>, <0.4 5 1>, <1 1 1>]}}
        )
       )
      )
      "Float (Alpha)"=0.03
     )
    )) {isSetParameters=true}
   &ScaleColor(
    Scale RGB=&Multiply_Vector(
     A=&SelectVectorFromArray(
      Direct Array Index=$System.RankedProgression
      Vector Selection Array=DataInterface() {propertyValues {FloatData [<3 2 1>, <1 1 0.5>, <1 1 0.2>, <1 1 0.5>, <1 0.5 1>, <0.5 1 1>, <2 0.2 0.1>, <1 0.1 1>]}}
     )
     B=&FN_Select_Vector_from_Time_Of_Day(
      Morning=<1 1 1>
      Day=<1 1 1>
      Evening=<1 1 1>
      Night=&SelectVectorFromArray(
       Direct Array Index=$System.RankedProgression
       Vector Selection Array=DataInterface() {propertyValues {FloatData [<0.5 0.5 1>, <1 1 1>, <0.5 0.5 1>, <1 1 1>, <1 1 1>, <1 1 1>, <0.4 5 1>, <1 1 1>]}}
      )
     )
    )
    ScaleA=true
    ScaleRGB=true
   )
  ]
  rendererClasses [&NiagaraSpriteRendererProperties]
  renderers [
   {rendererClass=&NiagaraSpriteRendererProperties}
   {rendererClass=&NiagaraSpriteRendererProperties, rendererIndex=1}
  ]
 }
}
summary {
 systemName=NS_Glider_Umbrella_RankedAurum_Glint
 emitters [
  {emitterName=Flare_loop001, rendererClasses [&NiagaraSpriteRendererProperties]}
  {emitterName=Flare_loop, rendererClasses [&NiagaraSpriteRendererProperties]}
 ]
 userVariables [
  User.bIsFrontEnd {type=NiagaraBool}
  User.bIsFrontEndPreview {type=NiagaraBool}
  User.ForwardDot {type=NiagaraFloat, defaultValue=0, description=""}
  User.PhysicalSurface {type=NiagaraInt32}
  User.RankedProgression {type=NiagaraInt32}
  User.SystemDelay {type=NiagaraFloat, defaultValue=0, description=""}
 ]
}
