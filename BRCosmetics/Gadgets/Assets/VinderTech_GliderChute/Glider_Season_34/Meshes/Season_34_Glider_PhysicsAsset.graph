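# Physics asset for the Season 34 glider cosmetic's skeletal mesh, a GliderChute-type gadget.
# It defines a single capsule body on the Frame bone, offset and rotated from the bone to approximate the glider as one coarse volume.
# It has no constraints, so it sets up no jointed ragdoll chain between bones.
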PhysicsAsset /BRCosmetics/Gadgets/Assets/VinderTech_GliderChute/Glider_Season_34/Meshes/Season_34_Glider_PhysicsAsset
bodies {Frame [{shapeType=Capsule, length=71.119, radius=61.2778, center=<1.497 -84.3811 -17.0989>, rotation=<rot 0 -8.2835 -67.1153>}]}
constraints []
